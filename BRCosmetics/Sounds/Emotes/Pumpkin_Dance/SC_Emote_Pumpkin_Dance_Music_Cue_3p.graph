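# Third-person music sound cue for the Pumpkin Dance emote.
# A mixer plays a one-shot intro immediately alongside a looping track held back by a fixed delay, so the loop follows the intro, with the intro mixed slightly louder.
# It uses emote attenuation, third-person emote music class and concurrency rules, keeps playing when silent, and sends to a music audio-analysis submix.
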SoundCue /BRCosmetics/Sounds/Emotes/Pumpkin_Dance/SC_Emote_Pumpkin_Dance_Music_Cue_3p
root=Mixer_0
nodes [
 {id=Delay_0, children [WavePlayer_2], properties {delayMax=3.67, delayMin=3.67}}
 {id=Mixer_0, children [WavePlayer_0,Delay_0], properties {inputVolume [1.25,1]}}
 {id=WavePlayer_0, properties {soundWaveAssetPtr=&/Emote_Pumpkin_Dance_Intro}}
 {id=WavePlayer_2, properties {bLooping=true, soundWaveAssetPtr=&/Emote_Pumpkin_Dance_Loop}}
]
properties {
 attenuationSettings=&Fort_Emotes_Athena_Att
 audioPropertiesSheet=/BRCosmetics/Sounds/Settings/PropertySheets/Emotes/APS_Cosmetic_Emote_Music_3P
 maxDistance=1600
 soundClassObject=/Game/Sounds/FortSoundClassesAndModes/Fort_Music_Emote_3P
 virtualizationMode=PlayWhenSilent
 volumeMultiplier=0.85
 concurrencySet [/Game/Sounds/Concurrency_Rules/Athena_EmoteMusic_Conc]
 soundSubmixSends [{minSendLevel=0, soundSubmix=&Fortnite_AudioAnalysis_Music}]
}
